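# A placeholder-simple Default Lit base material for the maze environment set, stored in an OLD_Maze folder.
# It is a Material, not a Material Instance, despite the MI_ prefix.
# Its graph holds only a dark blue-grey color constant and two scalar constants, with no textures, parameters or visible connections to material outputs.
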Material /Game/Environments/Sets/Maze/Materials/OLD_Maze/MI_Maze_BaseMaterial_01
graph {
 properties {pixelDepthOffsetMode=PDOM_Legacy, usedShadingModels=Default Lit}
 nodes [
  Constant3Vector() {constant=<rgba 0.0665 0.0705 0.099 1>}
  Constant() {r=0.35}
  Constant() {r=0.4}
 ]
}
