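# Default Lit test grass material in the OLD_Maze environment folder, built as a base Material despite its MI_ prefix.
# A scalar "TIle" parameter multiplies the texture coordinates, giving shared tiling for three Apollo terrain grass textures: diffuse, normal and mask.
# The diffuse color is multiplied by a constant muted olive-yellow tint, while the normal and mask samples are unmodified.
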Material /Game/Environments/Sets/Maze/Materials/OLD_Maze/MI_Maze_Grass_Test_01
graph {
 properties {pixelDepthOffsetMode=PDOM_Legacy, usedShadingModels=Default Lit}
 nodes [
  n1 = Multiply(A=ScalarParameter() {defaultValue=1, parameterName=TIle}, B=TextureCoordinate())
  Multiply(
   A=Constant3Vector() {constant=<rgba 0.5052 0.4838 0.2483 0>}
   B=TextureSample(UVs=@n1) {texture=/Game/Environments/Apollo/Sets/Terrain/Textures/T_Apollo_Grass_D_01}.RGB
  )
  TextureSample(UVs=@n1) {samplerType=SAMPLERTYPE_Normal, texture=/Game/Environments/Apollo/Sets/Terrain/Textures/T_Apollo_Grass_N_01}
  TextureSample(UVs=@n1) {samplerType=SAMPLERTYPE_Masks, texture=/Game/Environments/Apollo/Sets/Terrain/Textures/T_Apollo_Grass_S_01}
 ]
}
